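# Physics asset for the FuryFax axe, a melee weapon cosmetic, providing simplified collision for its mesh.
# It has a single body on the Melee_TwoHanded_Handle bone, made of one wide capsule offset from the bone origin.
# It defines no constraints, so there is no jointed or secondary motion between bodies.
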PhysicsAsset /BRCosmetics/Weapons/FORT_Melee/FuryFax/Meshes/FuryFax_Axe_PhysicsAsset
bodies {Melee_TwoHanded_Handle [{shapeType=Capsule, length=58.9299, radius=31.6116, center=<1.0056 5.7729 29.1037>, rotation=<rot -0.3087 -90 -0.0168>}]}
constraints []
